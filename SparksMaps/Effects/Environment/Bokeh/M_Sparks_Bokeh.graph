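# Translucent unlit Niagara sprite material for Sparks bokeh flares that pulse with the music through the audio-reactive BeatLifetime parameter.
# Particle color is scaled by a beat-driven emissive lerp, with optional stage-color tinting before emissive adjustments.
# Flare alpha sums exponential radial core and halo gradients, scaled by particle alpha and beat.
# Parameters can swap this procedural shape for a sparkle texture and add camera-distance fade.
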Material /SparksMaps/Effects/Environment/Bokeh/M_Sparks_Bokeh
graph {
 properties {
  bAutomaticallySetUsageInEditor=false
  bUsedWithNiagaraSprites=true
  blendMode=BLEND_Translucent
  shadingModel=MSM_Unlit
  usedShadingModels=Unlit
 }
 nodes [
  n1 = ParticleColor()
  n2 = Multiply(
   A=LinearInterpolate(
    A=ScalarParameter() {defaultValue=5, desc=BeatHigh, parameterName=Emis Beat Low}
    B=ScalarParameter() {defaultValue=1.5, desc=BeatLow, parameterName=Emis Beat High}
    Alpha=CollectionParameter() {collection=/FMAudioReactiveData/AudioAnalysis/MPC_Sparks_Audio_Parameters, parameterName=BeatLifetime}
   ) {desc=BeatData}
   B=@n1.RGB
  )
  n3 = Multiply(
   A=MaterialFunctionCall!(CustomUPosition) {materialFunction=/SparksStageShow/Systems/StageManager/FXManager/MaterialFunctions/MF_MojoStageColors}.Primary
   B=@n2
  )
  n4 = ScalarParameter() {parameterName=StageColor?}
  MaterialFunctionCall(Emissive=Multiply(A=LinearInterpolate(A=@n2, B=@n3, Alpha=@n4))) {materialFunction=&MF_FX_EmissiveAdjustments}
  n5 = Multiply(
   A=MaterialFunctionCall(
    Radius=ScalarParameter() {defaultValue=0.1, parameterName=Core Radius}
    Density=ScalarParameter() {defaultValue=5, desc=BeatHigh, parameterName=Core Density}
   ) {materialFunction=&RadialGradientExponential}.RadialGradientExponential
   B=ScalarParameter() {defaultValue=2, parameterName=Core Intensity}
  )
  n6 = Multiply(
   A=MaterialFunctionCall(
    Radius=ScalarParameter() {defaultValue=0.5, parameterName=Halo Radius}
    Density=ScalarParameter() {defaultValue=0.3, parameterName=Halo Density}
   ) {materialFunction=&RadialGradientExponential}.RadialGradientExponential
   B=ScalarParameter() {defaultValue=4, parameterName=Halo Glow}
  ) {constB=0.1}
  n7 = Saturate(None=Multiply(A=@n1.A, B=Add(A=@n5, B=@n6)))
  n8 = LinearInterpolate(
   A=ScalarParameter() {defaultValue=1, desc=BeatHigh, parameterName=Alpha Beat Low}
   B=ScalarParameter() {defaultValue=0.5, desc=BeatLow, parameterName=Alpha Beat High}
   Alpha=CollectionParameter() {collection=&MPC_Sparks_Audio_Parameters, parameterName=BeatLifetime}
  ) {desc=BeatData}
  n9 = ScalarParameter() {defaultValue=3000, parameterName=CamDist_SmoothStep_Max}
  Multiply(
   A=LinearInterpolate(
    A=Multiply(A=@n7, B=@n8) {desc=Material Flare}
    B=TextureSampleParameter2D() {desc=Texture Flare, parameterName=Texture, texture=/Game/Effects/Fort_Effects/Textures/Sparks/T_Sparkles_4Star_01_D}.RGB
    Alpha=ScalarParameter() {parameterName=UseTex?}
   ) {desc=Mat or Tex Flare?}
   B=LinearInterpolate(
    B=MaterialFunctionCall!(Distance=@n9) {materialFunction=/Game/Effects/Fort_Effects/Materials/_Functions/MF_CameraDistanceFade}.Result
    Alpha=ScalarParameter() {parameterName=CamDistanceFade?}
   ) {constA=1}
  )
  ScalarParameter() {defaultValue=3, parameterName=EmisBoost}
  TextureSample(UVs=Panner(Coordinate=TextureCoordinate() {uTiling=0.5, vTiling=0.5}) {speedX=0.1, speedY=0.1}) {samplerType=SAMPLERTYPE_LinearColor, texture=/Game/Effects/Niagara/MasterMaterials/Lightning/SplineLightning/T_SplineLightning_Noise}
 ]
}
